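# Unlit translucent particle sprite material for a round spark, with responsive AA enabled.
# Particle speed, divided by the ParticleSpeed parameter and clamped, drives a lerp on the T_SparkRound texture.
# Both lerp inputs read the red channel as wired, so speed does not change the result.
# A feature-level switch substitutes the green channel on ES2 and ES3_1, and the result is multiplied by particle alpha.
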Material /Figment_S08/Effects/Fort_Effects/Materials/M_SparkRound_SpeedLerp
graph {
 properties {
  bEnableResponsiveAA=true
  bUsedWithParticleSprites=true
  blendMode=BLEND_Translucent
  pixelDepthOffsetMode=PDOM_Legacy
  shadingModel=MSM_Unlit
  usedShadingModels=Unlit
 }
 nodes [
  n1 = TextureSample() {texture=/Figment_S08/Effects/Fort_Effects/Textures/Flares/T_SparkRound}
  n2 = ParticleSpeed()
  n3 = ScalarParameter() {defaultValue=100, parameterName=ParticleSpeed}
  n4 = LinearInterpolate(A=@n1.R, B=@n1.R, Alpha=Clamp(None=Divide(A=@n2, B=@n3)))
  Multiply(A=ParticleColor().A, B=FeatureLevelSwitch(Default=@n4, ES2=@n1.G, ES3_1=@n1.G, SM4_REMOVED=@n4))
 ]
}
